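# Default Lit material for the Rift spawner crystal, building an animated purple storm-cloud glow that darkens toward the base.
# Object-centred angle and world height form cylindrical UVs, twisted by an object-radius height gradient and panned.
# Two scales of cloud masks, shaped by that gradient, feed a subsurface-style colour boosted where surfaces face the camera.
# Blended, flattened cloud normals modulate it by view angle; a separate Fresnel term tints a dark blue base colour.
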Material /Game/Athena/Deimos/Spawners/RiftSpawners/Assets/M_RiftCrystal
graph {
 properties {pixelDepthOffsetMode=PDOM_Legacy, usedShadingModels=Default Lit}
 nodes [
  Constant()
  Constant()
  n1 = ObjectRadius()
  n2 = Subtract(A=ObjectPositionWS(), B=Multiply(A=@n1, B=Constant3Vector() {constant=<rgba 0 0 1 0>}))
  n3 = WorldPosition()
  n4 = OneMinus(None=SphereMask(
    A=MaterialFunctionCall(Float3=@n2) {materialFunction=&BreakOutFloat3Components}.B
    B=MaterialFunctionCall(Float3=@n3.XYZ) {materialFunction=&BreakOutFloat3Components}.B
    Radius=Multiply(A=@n1) {constB=2}
   ) {hardnessPercent=0})
  n5 = Multiply(A=@n4, B=ScalarParameter() {defaultValue=1, parameterName=Spin})
  n6 = WorldPosition() {worldPositionShaderOffset=WPT_ExcludeAllShaderOffsets}
  n7 = ComponentMask(None=Subtract(A=@n6.XYZ, B=ObjectPositionWS())) {g=true, r=true}
  n8 = ScalarParameter() {defaultValue=1304.1251, parameterName=scale}
  n9 = AppendVector(
   A=Multiply(
    A=MaterialFunctionCall(Vector or UVs=@n7) {materialFunction=&VectorToRadialValue}.Vector Converted to Angle
    B=ScalarParameter() {defaultValue=1, parameterName=scale_1}
   )
   B=ComponentMask(None=Divide(A=@n6.XYZ, B=@n8)) {b=true}
  )
  n10 = ScalarParameter() {defaultValue=-0.1, parameterName=Speed}
  n11 = Panner(Coordinate=Add(A=@n5, B=Multiply(A=@n9) {constB=2}), Speed=@n10) {bFractionalPart=true, speedY=-0.1}
  n12 = TextureSample(UVs=@n11) {samplerType=SAMPLERTYPE_Normal, texture=/Game/Global/Textures/Default/Tileable_Patterns/T_StormCloudDetail01_N}
  n13 = ScalarParameter() {defaultValue=0.444, parameterName=flatnormal}
  MaterialFunctionCall(Normal=@n12.RGB, Flatness=@n13) {materialFunction=&FlattenNormal}
  Multiply(
   A=Fresnel(ExponentIn=ScalarParameter() {defaultValue=1.007, parameterName=Fresexp}, Normal=VertexNormalWS())
   B=VectorParameter() {parameterName=BaseColor, defaultValue=<rgba 0.0211 0.0073 0.1 1>}.RGB
  )
  n14 = Panner(Coordinate=Add(A=@n9, B=@n5), Speed=@n10) {bFractionalPart=true, speedY=-0.1}
  n15 = TextureSample(UVs=@n14) {samplerType=SAMPLERTYPE_LinearGrayscale, texture=/Game/Global/Textures/Default/Tileable_Patterns/T_StormCloudDetail01_M}
  n16 = ScalarParameter() {defaultValue=0.2, parameterName=FineCloudModulation}
  n17 = LinearInterpolate(
   A=Subtract(B=@n16) {constA=0}
   B=@n16
   Alpha=TextureSample(UVs=@n11) {samplerType=SAMPLERTYPE_LinearGrayscale, texture=&T_StormCloudDetail01_M}.R
  )
  n18 = Power(Base=@n4, Exp=ScalarParameter() {defaultValue=4.2789, parameterName=BAseDarkness})
  n19 = Multiply(A=VectorParameter() {parameterName=SubSurfaceColor, defaultValue=<rgba 5.2122 0 8 1>}.RGB, B=Multiply(A=Add(A=@n15.R, B=@n17), B=@n18))
  n20 = VertexNormalWS()
  n21 = CameraVectorWS()
  n22 = Multiply(A=@n19, B=Power(Base=DotProduct(A=@n20, B=@n21), Exp=ScalarParameter() {defaultValue=16, parameterName=SSS_RimExp}))
  n23 = ScalarParameter() {defaultValue=1, parameterName=SSS_RimIntensity}
  n24 = TextureSample(UVs=@n14) {samplerType=SAMPLERTYPE_Normal, texture=&T_StormCloudDetail01_N}
  n25 = Multiply(A=@n12.RGB, B=Constant3Vector() {constant=<rgba 1 1 0 0>})
  n26 = Transform(None=MaterialFunctionCall(Normal=Add(A=@n24.RGB, B=@n25), Flatness=@n13) {materialFunction=&FlattenNormal}.Result)
  n27 = CameraVectorWS()
  Multiply(
   A=Add(A=@n19, B=Multiply(A=@n22, B=@n23))
   B=Power(Base=DotProduct(A=@n26, B=@n27), Exp=ScalarParameter() {defaultValue=4, parameterName=fsafsaf})
  )
  Multiply(A=@n15.R, B=@n18)
  TextureCoordinate() {uTiling=2, vTiling=2}
  TextureSampleParameter2D() {parameterName=Normal, samplerType=SAMPLERTYPE_Normal, texture=/Game/Environments/Props/Rocks/Textures/StoneLarge_01_N}
 ]
}
